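; BP_AssassinSuitMark is a short-lived state marker actor that toggles a "Hide Element 01" material parameter for an assassin suit cosmetic effect.
; CE_HideBullet stores the passed dynamic material instance, sets the parameter to 1, then destroys the actor after a retriggerable delay of the given duration.
; EventDestroyed resets the parameter to 0, tying the hidden state to the actor's lifetime.
; The overlap event and construction script are empty.
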(event Collision|EventActorBeginOverlap
  (params (OtherActor :type Actor%20Object%20Reference)))

(event Custom|CE_HideBullet
  (params (MID :type Material%20Instance%20Dynamic%20Object%20Reference) (Duration :type Float%20%28double-precision%29))
  (Variables|Default|SetMID MID)
  (Rendering|Material|SetScalarParameterValue MID "Hide Element 01" 1.0 :!class /Script/Engine.MaterialInstanceDynamic)
  (Utilities|FlowControl|RetriggerableDelay Duration)
  (Actor|DestroyActor))

(event EventDestroyed
  (Rendering|Material|SetScalarParameterValue (Variables|Default|GetMID) "Hide Element 01" 0.0 :!class /Script/Engine.MaterialInstanceDynamic))

(fn ConstructionScript)
